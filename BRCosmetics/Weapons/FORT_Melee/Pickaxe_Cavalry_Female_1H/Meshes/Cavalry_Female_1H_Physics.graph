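# Physics asset for the Cavalry Female one-handed pickaxe cosmetic, providing collision bodies for its handle and a small dangling ornament.
# A capsule covers the handle, two short capsules form the rope, and three thin boxes form the dangles.
# All three dangles constrain to the lower rope segment, which constrains to the upper one with tighter angular limits.
# No constraint links the rope to the handle body.
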PhysicsAsset /BRCosmetics/Weapons/FORT_Melee/Pickaxe_Cavalry_Female_1H/Meshes/Cavalry_Female_1H_Physics
bodies {
 Melee_TwoHanded_Handle [{shapeType=Capsule, length=37.6945, radius=2.9806, center=<0.0472 0.1853 5.3546>, rotation=<rot -0.8822 -91.5822 0.1996>}]
 dyn_dangle_01 [{shapeType=Box, extentX=0.5004, extentY=2.0839, extentZ=4.9914, center=<0.1069 0 -2.6098>}]
 dyn_dangle_02 [{shapeType=Box, extentX=0.5004, extentY=2.0839, extentZ=6.2753, center=<0 0 -3.1178>}]
 dyn_dangle_03 [{shapeType=Box, extentX=0.5004, extentY=2.0839, extentZ=5.0478, center=<-0.1726 0 -2.6785>, rotation=<rot 1.4868 0 0>}]
 dyn_rope_01 [{shapeType=Capsule, length=0.8981, radius=1.1757, center=<-0.0899 -0.0077 -0.6598>, rotation=<rot 0 89.9976 -0.6437>}]
 dyn_rope_02 [{shapeType=Capsule, length=1.0925, radius=0.9167, center=<-0.0506 -0.0076 -0.735>, rotation=<rot 0 -89.9995 2.5317>}]
}
constraints [
 {bone1Name=dyn_dangle_01, bone2Name=dyn_rope_02, swing1LimitDegrees=2, swing2LimitDegrees=20, twistLimitDegrees=45}
 {bone1Name=dyn_dangle_02, bone2Name=dyn_rope_02, swing1LimitDegrees=2, swing2LimitDegrees=20, twistLimitDegrees=45}
 {bone1Name=dyn_dangle_03, bone2Name=dyn_rope_02, swing1LimitDegrees=2, swing2LimitDegrees=20, twistLimitDegrees=45}
 {bone1Name=dyn_rope_02, bone2Name=dyn_rope_01, swing1LimitDegrees=5, swing2LimitDegrees=5, twistLimitDegrees=30}
]
